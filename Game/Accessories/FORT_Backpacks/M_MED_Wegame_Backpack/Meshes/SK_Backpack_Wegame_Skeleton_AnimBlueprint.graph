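# Animation Blueprint for the Wegame backpack skeleton that applies a rotation adjustment to the backpack's root bone.
# The input pose passes through an additive, bone-space Transform (Modify) Bone on backpack_root at full alpha, then goes to the output pose.
# The rotation comes from BackpackRotation, which is not declared in this asset; the event graph is empty.
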AnimBlueprint /Game/Accessories/FORT_Backpacks/M_MED_Wegame_Backpack/Meshes/SK_Backpack_Wegame_Skeleton_AnimBlueprint
target_skeleton=&/SK_Backpack_Wegame_Skeleton
graphs [
 AnimGraph {nodes [
   AnimGraphNode_SubInput_0 = Input Pose()
   n1 = "Transform (Modify) Bone - Bone: backpack_root"(
    ComponentPose=Local To Component(LocalPose=@AnimGraphNode_SubInput_0)
    Alpha=1
    Translation="0,0,0"
    Rotation=Get BackpackRotation() -> (BackpackRotation:Rotator)
    Scale="1,1,1"
   ) {rotationMode=BMM_Additive, rotationSpace=BCS_BoneSpace, alphaBoolBlend {blendOption=Linear}, boneToModify {boneName=backpack_root}}
   Output Pose(Result=Component To Local(ComponentPose=@n1)) {layerGroup=None}
  ]}
 EventGraph {links [], nodes []}
]
variables []
